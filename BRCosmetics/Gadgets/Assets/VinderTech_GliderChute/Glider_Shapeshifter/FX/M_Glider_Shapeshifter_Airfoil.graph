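# Additive, unlit, two-sided material for the Shapeshifter glider's airfoil effect, flagged for Niagara ribbons.
# It samples one panning caustics texture three times at offset UVs, packing the results into RGB for chromatic fringing.
# Particle color tints that pattern through an emissive adjustment function.
# Opacity multiplies the pattern by particle alpha, a U gradient and square-edge masks, then applies near-camera and depth fades.
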Material /BRCosmetics/Gadgets/Assets/VinderTech_GliderChute/Glider_Shapeshifter/FX/M_Glider_Shapeshifter_Airfoil
graph {
 properties {
  bUsedWithNiagaraRibbons=true
  blendMode=BLEND_Additive
  pixelDepthOffsetMode=PDOM_Legacy
  shadingModel=MSM_Unlit
  twoSided=true
  usedShadingModels=Unlit
 }
 nodes [
  n1 = ParticleColor()
  n2 = TextureCoordinate() {uTiling=0.1}
  n3 = ComponentMask(None=VectorParameter() {parameterName=ChromaticOffset, defaultValue=<rgba 0.01 0.01 0 0>}.RGB) {g=true, r=true}
  n4 = Add(A=@n2, B=@n3)
  n5 = ComponentMask(None=VectorParameter() {parameterName=Panner Speed, defaultValue=<rgba -0.25 0 0 0>}.RGB) {g=true, r=true}
  n6 = Add(A=@n2, B=Multiply(A=@n3) {constB=-1})
  n7 = MaterialFunctionCall(
   X=TextureSample(UVs=Panner(Coordinate=@n4, Speed=@n5)) {texture=/Game/Global/Textures/Default/Caustics/T_Caustics_Softer}.R
   Y=TextureSample(UVs=Panner(Coordinate=@n2, Speed=@n5)) {texture=&T_Caustics_Softer}.R
   Z=TextureSample(UVs=Panner(Coordinate=@n6, Speed=@n5)) {texture=&T_Caustics_Softer}.R
  ) {materialFunction=&MakeFloat3}
  n8 = OneMinus(None=ComponentMask(None=TextureCoordinate()) {r=true})
  n9 = MaterialFunctionCall!(UVs, Mult=ScalarParameter() {defaultValue=4, parameterName=EdgeMult}, Exp=ScalarParameter() {defaultValue=2, parameterName=EdgeExp}) {materialFunction=/Game/Effects/Fort_Effects/MaterialFunctions/MF_SquareEdgeMask}
  n10 = Multiply(A=@n9.HorizMask, B=@n9.HorizMask)
  n11 = Power(Base=@n9.LeftMask) {constExponent=0.75}
  n12 = Multiply(A=@n7.Result, B=Multiply(A=Saturate(None=@n8), B=Multiply(A=@n10, B=@n11)))
  n13 = ScalarParameter() {defaultValue=300, parameterName=CamFade}
  DepthFade(
   Opacity=Multiply(
    A=Multiply(A=@n1.A, B=@n12)
    B=MaterialFunctionCall(Distance=@n13) {materialFunction=&Near_Camera_Fade}.Result
   )
   FadeDistance=ScalarParameter() {defaultValue=20, parameterName=DepthFadeDist}
  )
  MaterialFunctionCall(Emissive=Multiply(A=@n7.Result, B=@n1.RGB)) {materialFunction=&MF_FX_EmissiveAdjustments}
 ]
}
